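# Physics asset for the Weeping Woods backpack cosmetic, defining capsule bodies and joints for its "honey" dyn bones.
# A large capsule covers the pack bone and anchors four chains of small capsules: two-link back and right chains, plus single left and front bones.
# Angle-limited constraints allow moderate swing with very little twist, with the front bone held tighter on its second swing axis.
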PhysicsAsset /BRCosmetics/Accessories/FORT_Backpacks/M_MED_WeepingWoods_Backpack/Meshes/M_MED_Weeping_Woods_Pack_Physics
bodies {
 dyn_honey_bk_01 [{shapeType=Capsule, length=1.5, radius=1.2, center=<1.5 0 0>, rotation=<rot 90 0 0>}]
 dyn_honey_bk_02 [{shapeType=Capsule, length=1.5, radius=1.2, center=<1.5 0 0>, rotation=<rot 90 0 0>}]
 dyn_honey_ft [{shapeType=Capsule, length=1.5, radius=1.2, center=<1.5 0 0>, rotation=<rot 90 0 0>}]
 dyn_honey_l_01 [{shapeType=Capsule, length=1.5, radius=1.2, center=<2.2 0 0>, rotation=<rot 90 0 0>}]
 dyn_honey_r_01 [{shapeType=Capsule, length=1.5, radius=1.2, center=<1.5 0 0>, rotation=<rot 90 0 0>}]
 dyn_honey_r_02 [{shapeType=Capsule, length=1.5, radius=1.2, center=<1.5 0 0>, rotation=<rot 90 0 0>}]
 pack [{shapeType=Capsule, length=14, radius=11, center=<-0.3 -9 -7>}]
}
constraints [
 {bone1Name=dyn_honey_bk_01, bone2Name=pack, swing1LimitDegrees=45, swing2LimitDegrees=45, twistLimitDegrees=5}
 {bone1Name=dyn_honey_bk_02, bone2Name=dyn_honey_bk_01, swing1LimitDegrees=45, swing2LimitDegrees=45, twistLimitDegrees=5}
 {bone1Name=dyn_honey_l_01, bone2Name=pack, swing1LimitDegrees=45, swing2LimitDegrees=45, twistLimitDegrees=5}
 {bone1Name=dyn_honey_r_01, bone2Name=pack, swing1LimitDegrees=45, swing2LimitDegrees=45, twistLimitDegrees=5}
 {bone1Name=dyn_honey_r_02, bone2Name=dyn_honey_r_01, swing1LimitDegrees=45, swing2LimitDegrees=45, twistLimitDegrees=5}
 {bone1Name=dyn_honey_ft, bone2Name=pack, swing1LimitDegrees=45, swing2LimitDegrees=10, twistLimitDegrees=5}
]
